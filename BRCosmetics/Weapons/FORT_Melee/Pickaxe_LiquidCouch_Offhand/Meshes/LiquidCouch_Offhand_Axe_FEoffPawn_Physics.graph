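# Physics asset for the LiquidCouch offhand pickaxe mesh, which carries a small character-like skeleton with spine, limbs, head and jaw.
# Capsule bodies cover each bone, chained from the two-handed handle through both legs and the spine, with arms and head branching from the top spine bone.
# Constraints use very tight swing and twist limits for a stiff wobble, while the jaw locks twist and swings on one axis.
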PhysicsAsset /BRCosmetics/Weapons/FORT_Melee/Pickaxe_LiquidCouch_Offhand/Meshes/LiquidCouch_Offhand_Axe_FEoffPawn_Physics
bodies {
 Melee_TwoHanded_Handle [{shapeType=Capsule, length=8.2833, radius=2.3971, shapeName=Melee_TwoHanded_Handle_capsule, center=<-0.4992 -0.3833 0.152>}]
 arm_l_01 [{shapeType=Capsule, length=2.4941, radius=2.1873, shapeName=arm_l_01_capsule, center=<1.9706 0 0>, rotation=<rot 0 90 90>}]
 arm_l_02 [{shapeType=Capsule, length=2.6365, radius=2.1918, shapeName=arm_l_02_capsule, center=<2.2774 0 0>, rotation=<rot 0 90 90>}]
 arm_l_03 [{shapeType=Capsule, length=2.516, radius=1.8277, shapeName=arm_l_03_capsule, center=<1.9212 0 0>, rotation=<rot 0 90 90>}]
 arm_l_04 [{shapeType=Capsule, length=1.6522, radius=2.0462, shapeName=arm_l_04_capsule, center=<2.2822 0 0>, rotation=<rot 0 90 90>}]
 arm_r_01 [{shapeType=Capsule, length=2.4933, radius=2.1972, shapeName=arm_r_01_capsule, center=<-3.3916 0 0>, rotation=<rot 90 0 0>}]
 arm_r_02 [{shapeType=Capsule, length=2.2893, radius=1.8952, shapeName=arm_r_02_capsule, center=<-3.0393 0 0>, rotation=<rot 90 0 0>}]
 arm_r_03 [{shapeType=Capsule, length=2.2987, radius=2.0871, shapeName=arm_r_03_capsule, center=<-3.3663 0 0>, rotation=<rot 90 0 0>}]
 arm_r_04 [{shapeType=Capsule, length=2.4739, radius=2.3757, shapeName=arm_r_04_capsule, center=<-3.5952 0 0>, rotation=<rot 90 0 0>}]
 head [
  {shapeType=Capsule, length=6, radius=2, shapeName=head_capsule, center=<2.1866 0 0.7141>, rotation=<rot 5 -180 180>}
  {shapeType=Capsule, length=5, radius=2, shapeName=head_capsule, center=<1.5788 -3.6233 1.208>, rotation=<rot -9.9997 0 0>}
 ]
 jaw [{shapeType=Capsule, length=4.178, radius=4.5955, shapeName=jaw_capsule, center=<3.6959 0 4.499>, rotation=<rot 80 -179.9992 -179.9993>}]
 leg_l_01 [{shapeType=Capsule, length=2.5427, radius=1.7173, shapeName=leg_l_01_capsule, center=<3.1005 0 0>, rotation=<rot 90 0 0>}]
 leg_l_02 [{shapeType=Capsule, length=1.4761, radius=2.0465, shapeName=leg_l_02_capsule, center=<2.4313 0 0>, rotation=<rot 90 0 0>}]
 leg_l_03 [{shapeType=Capsule, length=1.7943, radius=2.2423, shapeName=leg_l_03_capsule, center=<2.8386 0 0>, rotation=<rot 90 0 0>}]
 leg_l_04 [{shapeType=Capsule, length=1.554, radius=2.2647, shapeName=leg_l_04_capsule, center=<2.7891 0 0>, rotation=<rot 90 0 0>}]
 leg_r_01 [{shapeType=Capsule, length=1.865, radius=2.0792, shapeName=leg_r_01_capsule, center=<2.4086 0 0>, rotation=<rot 90 -0.0001 0>}]
 leg_r_02 [{shapeType=Capsule, length=3.1929, radius=2.0532, shapeName=leg_r_02_capsule, center=<3.3026 0 0>, rotation=<rot 90 -0.0001 0>}]
 leg_r_03 [{shapeType=Capsule, length=1, radius=2.505, shapeName=leg_r_03_capsule, center=<2.414 0 0>, rotation=<rot 90 0 0>}]
 leg_r_04 [{shapeType=Capsule, length=1.0632, radius=2.7659, shapeName=leg_r_04_capsule, center=<2.6991 0 0>, rotation=<rot 90 0 0>}]
 spine_00 [{shapeType=Capsule, length=5.9444, radius=3.2623, shapeName=spine_00_capsule, center=<2.2354 0.0369 0.3213>, rotation=<rot 0.8411 -90 0.1679>}]
 spine_01 [{shapeType=Capsule, length=1.7033, radius=4.7826, shapeName=spine_01_capsule}]
 spine_02 [{shapeType=Capsule, length=1.655, radius=5.129, shapeName=spine_02_capsule}]
 spine_03 [{shapeType=Capsule, length=1.8531, radius=4.2102, shapeName=spine_03_capsule}]
 spine_04 [{shapeType=Capsule, length=1.6994, radius=4.2283, shapeName=spine_04_capsule}]
}
constraints [
 {bone1Name=leg_l_01, bone2Name=Melee_TwoHanded_Handle, swing1LimitDegrees=5, swing2LimitDegrees=5, twistLimitDegrees=1}
 {bone1Name=leg_l_02, bone2Name=leg_l_01, swing1LimitDegrees=5, swing2LimitDegrees=5, twistLimitDegrees=5}
 {bone1Name=leg_l_03, bone2Name=leg_l_02, swing1LimitDegrees=5, swing2LimitDegrees=5, twistLimitDegrees=5}
 {bone1Name=leg_l_04, bone2Name=leg_l_03, swing1LimitDegrees=5, swing2LimitDegrees=5, twistLimitDegrees=5}
 {bone1Name=leg_r_01, bone2Name=Melee_TwoHanded_Handle, swing1LimitDegrees=5, swing2LimitDegrees=5, twistLimitDegrees=1}
 {bone1Name=leg_r_02, bone2Name=leg_r_01, swing1LimitDegrees=5, swing2LimitDegrees=5, twistLimitDegrees=5}
 {bone1Name=leg_r_03, bone2Name=leg_r_02, swing1LimitDegrees=5, swing2LimitDegrees=5, twistLimitDegrees=5}
 {bone1Name=leg_r_04, bone2Name=leg_r_03, swing1LimitDegrees=5, swing2LimitDegrees=5, twistLimitDegrees=5}
 {bone1Name=spine_01, bone2Name=spine_00, swing1LimitDegrees=2, swing2LimitDegrees=3, twistLimitDegrees=1}
 {bone1Name=spine_02, bone2Name=spine_01, swing1LimitDegrees=2, swing2LimitDegrees=3, twistLimitDegrees=1}
 {bone1Name=spine_03, bone2Name=spine_02, swing1LimitDegrees=2, swing2LimitDegrees=3, twistLimitDegrees=1}
 {bone1Name=spine_04, bone2Name=spine_03, swing1LimitDegrees=2, swing2LimitDegrees=3, twistLimitDegrees=1}
 {bone1Name=arm_l_01, bone2Name=spine_04, swing1LimitDegrees=3, swing2LimitDegrees=3, twistLimitDegrees=1}
 {bone1Name=arm_l_02, bone2Name=arm_l_01, swing1LimitDegrees=3, swing2LimitDegrees=3, twistLimitDegrees=1}
 {bone1Name=arm_l_03, bone2Name=arm_l_02, swing1LimitDegrees=3, swing2LimitDegrees=3, twistLimitDegrees=1}
 {bone1Name=arm_l_04, bone2Name=arm_l_03, swing1LimitDegrees=3, swing2LimitDegrees=3, twistLimitDegrees=1}
 {bone1Name=arm_r_01, bone2Name=spine_04, swing1LimitDegrees=3, swing2LimitDegrees=3, twistLimitDegrees=1}
 {bone1Name=arm_r_02, bone2Name=arm_r_01, swing1LimitDegrees=3, swing2LimitDegrees=3, twistLimitDegrees=1}
 {bone1Name=arm_r_03, bone2Name=arm_r_02, swing1LimitDegrees=3, swing2LimitDegrees=3, twistLimitDegrees=1}
 {bone1Name=arm_r_04, bone2Name=arm_r_03, swing1LimitDegrees=3, swing2LimitDegrees=3, twistLimitDegrees=1}
 {bone1Name=head, bone2Name=spine_04, swing1LimitDegrees=3, swing2LimitDegrees=3, twistLimitDegrees=1}
 {bone1Name=jaw, bone2Name=head, swing1LimitDegrees=0, swing2LimitDegrees=10, twistLimitDegrees=45, twistMotion=Locked}
 {bone1Name=spine_00, bone2Name=Melee_TwoHanded_Handle, swing1LimitDegrees=2, swing2LimitDegrees=3, twistLimitDegrees=1}
]
